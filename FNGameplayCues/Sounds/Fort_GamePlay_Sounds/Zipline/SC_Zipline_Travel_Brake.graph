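# Sound cue for the brake sound during zipline travel, built from brake-start scrape and squeak samples.
# A mixer layers two modulated branches: one randomly picks among seven scrape waves, the other among six squeak waves, all equally weighted.
# The squeak modulator raises the maximum volume, and the mixer sets separate levels per layer.
# It uses the zipline travel attenuation and the gadgets/items sound class.
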SoundCue /FNGameplayCues/Sounds/Fort_GamePlay_Sounds/Zipline/SC_Zipline_Travel_Brake
root=Mixer_0
nodes [
 {id=Mixer_0, children [Modulator_0,Modulator_1], properties {inputVolume [0.7,0.85]}}
 {id=Modulator_0, children [Random_0]}
 {id=Modulator_1, children [Random_1], properties {volumeMax=1.35}}
 {
  id=Random_0
  children [WavePlayer_1,WavePlayer_3,WavePlayer_4,WavePlayer_6,WavePlayer_7,WavePlayer_8,WavePlayer_9]
  properties {weights [1,1,1,1,1,1,1]}
 }
 {
  id=Random_1
  children [WavePlayer_10, WavePlayer_11, WavePlayer_12, WavePlayer_13, WavePlayer_14, WavePlayer_15]
  properties {weights [1,1,1,1,1,1]}
 }
 {id=WavePlayer_1, properties {soundWaveAssetPtr=&/Zipline_Brake_Start_Scrape_02}}
 {id=WavePlayer_10, properties {soundWaveAssetPtr=&/Zipline_Brake_Start_Squeak_01}}
 {id=WavePlayer_11, properties {soundWaveAssetPtr=&/Zipline_Brake_Start_Squeak_02}}
 {id=WavePlayer_12, properties {soundWaveAssetPtr=&/Zipline_Brake_Start_Squeak_03}}
 {id=WavePlayer_13, properties {soundWaveAssetPtr=&/Zipline_Brake_Start_Squeak_04}}
 {id=WavePlayer_14, properties {soundWaveAssetPtr=&/Zipline_Brake_Start_Squeak_05}}
 {id=WavePlayer_15, properties {soundWaveAssetPtr=&/Zipline_Brake_Start_Squeak_06}}
 {id=WavePlayer_3, properties {soundWaveAssetPtr=&/Zipline_Brake_Start_Scrape_03}}
 {id=WavePlayer_4, properties {soundWaveAssetPtr=&/Zipline_Brake_Start_Scrape_04}}
 {id=WavePlayer_6, properties {soundWaveAssetPtr=&/Zipline_Brake_Start_Scrape_05}}
 {id=WavePlayer_7, properties {soundWaveAssetPtr=&/Zipline_Brake_Start_Scrape_06}}
 {id=WavePlayer_8, properties {soundWaveAssetPtr=&/Zipline_Brake_Start_Scrape_07}}
 {id=WavePlayer_9, properties {soundWaveAssetPtr=&/Zipline_Brake_Start_Scrape_08}}
]
properties {
 attenuationSettings=/Game/Sounds/Attenuation/Fort_Zipline_Travel_Att
 maxDistance=12700
 soundClassObject=&Fort_Gadgets_Items
}
