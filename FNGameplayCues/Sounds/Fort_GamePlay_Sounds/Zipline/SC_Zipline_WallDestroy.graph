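# Sound cue for a zipline wall-destroy impact, reusing Carmine punch-blast waves in the gadgets and items sound class.
# A quality-level node routes its first two tiers to the full wave with randomized pitch and volume, and the last two to an "_opt" wave with fixed modulator pitch.
# The cue raises overall pitch, lowers volume, uses zipline attachment attenuation, and overrides concurrency to cap simultaneous instances with volume scaling.
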SoundCue /FNGameplayCues/Sounds/Fort_GamePlay_Sounds/Zipline/SC_Zipline_WallDestroy
root=QualityLevel_0
nodes [
 {id=Modulator_0, children [WavePlayer_2], properties {pitchMax=1.1, pitchMin=0.85, volumeMin=0.7}}
 {id=Modulator_2, children [WavePlayer_0], properties {pitchMax=1, pitchMin=1}}
 {id=QualityLevel_0, children [Modulator_0,Modulator_0,Modulator_2,Modulator_2]}
 {id=WavePlayer_0, properties {soundWaveAssetPtr=/Game/Sounds/Fort_GamePlay_Sounds/Carmine/Carmine_Punch_Blast_02_opt}}
 {id=WavePlayer_2, properties {soundWaveAssetPtr=/Game/Sounds/Fort_GamePlay_Sounds/Carmine/Carmine_Punch_Blast_02}}
]
properties {
 attenuationSettings=/Game/Sounds/Attenuation/Fort_Zipline_Attachment_Att
 bOverrideConcurrency=true
 maxDistance=10000
 pitchMultiplier=1.2
 soundClassObject=&Fort_Gadgets_Items
 volumeMultiplier=0.55
 concurrencyOverrides {maxCount=3, volumeScale=0.5}
}
